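# Translucent UI material for a pickup background in the Battle Royale HUD, going by its name and path.
# MF_UI_Box supplies a full-size box fill, which is multiplied by an inverted horizontal LinearGradient so it fades from left to right.
# An Opacity scalar parameter scales the faded mask, alongside a constant left at its default of zero.
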Material /BattleRoyaleGameUI/HUD/PickupManager/Assets/M_UI_PickUp_Bg
graph {
 properties {bAutomaticallySetUsageInEditor=false, blendMode=BLEND_Translucent, materialDomain=MD_UI, usedShadingModels=Default Lit}
 nodes [
  Constant()
  n1 = Constant2Vector() {g=1, r=1}
  n2 = MaterialFunctionCall() {materialFunction=&LinearGradient}
  Multiply(
   A=Multiply(A=MaterialFunctionCall(Size=@n1) {materialFunction=&MF_UI_Box}.Fill, B=OneMinus(None=@n2.UGradient))
   B=ScalarParameter() {defaultValue=0.5, parameterName=Opacity}
  )
 ]
}
