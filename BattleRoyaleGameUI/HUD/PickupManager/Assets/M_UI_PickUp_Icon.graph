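# M_UI_PickUp_Icon is a translucent UI-domain material in the Battle Royale HUD's PickupManager assets, named for pickup icons.
# Its graph holds a single TextureSample node, with no texture assignment or connection to the material outputs shown.
# Automatic usage setting in the editor is turned off.
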Material /BattleRoyaleGameUI/HUD/PickupManager/Assets/M_UI_PickUp_Icon
graph {
 properties {bAutomaticallySetUsageInEditor=false, blendMode=BLEND_Translucent, materialDomain=MD_UI, usedShadingModels=Default Lit}
 nodes [TextureSample()]
}
