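# Translucent UI material that draws the resource background in the pickup manager HUD, with an animated selected state.
# An eased IsSelected value drives a perspective-warped trapezoid SDF, morphing its shape as selection changes.
# Opacity blends the shape's fill with a soft SDF glow, scaled by IsSelected, with a low-end fallback for fill opacity.
# Color comes from shared adaptive colors, darkened by a rotated, offset rounded-box shadow with animatable strength.
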Material /BattleRoyaleGameUI/HUD/PickupManager/Assets/M_UI_ResourceBackground
graph {
 properties {bAutomaticallySetUsageInEditor=false, blendMode=BLEND_Translucent, materialDomain=MD_UI, usedShadingModels=Default Lit}
 nodes [
  n1 = MaterialFunctionCall() {materialFunction=&GetUserInterfaceUV}
  n2 = AppendVector(A=ScalarParameter() {defaultValue=-0.02, parameterName=X_Offset}, B=ScalarParameter() {defaultValue=0.05, parameterName=Y_Offset})
  n3 = ScalarParameter() {defaultValue=0.05, parameterName=PerspectiveRotation}
  n4 = NamedRerouteUsage()
  n5 = ScalarParameter() {defaultValue=0.085, parameterName=PerspectiveAmount}
  n6 = ScalarParameter() {defaultValue=0.825, parameterName=ShapeBottomWidth}
  n7 = ScalarParameter() {defaultValue=0.9, parameterName=ShapeTopWidth}
  n8 = NamedRerouteUsage()
  n9 = MaterialFunctionCall("Value (0-1)"=ScalarParameter() {defaultValue=1, parameterName=IsSelected}) {materialFunction=&MF_UI_EaseCurves}
  n10 = MaterialFunctionCall!(
   UVs=MaterialFunctionCall!(
    UV=Subtract(A=@n1.Normalized UV, B=@n2)
    Perspective Rotation=LinearInterpolate(B=@n3, Alpha=@n4)
    Perspective Amount=LinearInterpolate(B=@n5, Alpha=@n4)
   ) {materialFunction=/FortUILibrary/MaterialFunctions/Basics/MF_UI_Perspective}.Perspective UVs
   TopWidth=LinearInterpolate(A=@n6, B=LinearInterpolate(B=@n7, Alpha=@n8) {constA=0.75}, Alpha=@n9.EaseInSine)
   BottomWidth=LinearInterpolate(B=@n6, Alpha=NamedRerouteUsage()) {constA=0.75}
   Height=LinearInterpolate(B=ScalarParameter() {defaultValue=0.825, parameterName=ShapeHeight}, Alpha=NamedRerouteUsage()) {constA=0.75}
   OutlineThickness
   OutlineAlignment
   GlowMin
   GlowMax
  ) {materialFunction=/FortUILibrary/MaterialFunctions/Shapes/MF_UI_IsoscelesTrapezoid}
  n11 = MaterialFunctionCall(
   DistanceField=@n10.SDF
   Threshold=ScalarParameter() {defaultValue=-0.1, parameterName=GlowThreshold}
   Smoothing=ScalarParameter() {defaultValue=0.5, parameterName=GlowSmoothing}
  ) {materialFunction=&MF_UI_AntiAliasing}
  n12 = ScalarParameter() {defaultValue=0.5, parameterName=SelectedGlowOpacity}
  n13 = ScalarParameter() {defaultValue=1, parameterName=SelectedOpacity}
  n14 = Constant() {r=0.9}
  LinearInterpolate(
   B=LinearInterpolate(
    A=Multiply(A=@n11.Inner, B=@n12)
    B=MaterialFunctionCall!(Default=@n13, LowEnd=@n14) {materialFunction=/BattleRoyaleGameUI/Shared/MF_UI_LowEnd_ValueSwitcher}.Result
    Alpha=@n10.Fill
   )
   Alpha=ScalarParameter() {defaultValue=1, parameterName=IsSelected}
  )
  n15 = MaterialFunctionCall!(BufferSource) {materialFunction=/BattleRoyaleGameUI/Shared/MF_UI_AdaptiveColors}
  n16 = ComponentMask(None=MaterialFunctionCall() {materialFunction=&GetUserInterfaceUV}.Normalized UV) {g=true}
  n17 = LinearInterpolate(
   A=MaterialFunctionCall(Default=@n15.Dim, LowEnd=@n15.Dim) {materialFunction=&MF_UI_LowEnd_ValueSwitcher}.Result
   B=LinearInterpolate(A=@n15.Dim, B=@n15.Dim, Alpha=Saturate(None=@n16))
   Alpha=@n10.Fill
  )
  n18 = ScalarParameter() {defaultValue=0.25, parameterName=ShadowOpacity}
  n19 = ScalarParameter() {defaultValue=1, parameterName=Anim_Shadow}
  n20 = TextureCoordinate()
  n21 = ScalarParameter() {defaultValue=0.1625, parameterName=ShadowRotation}
  n22 = ScalarParameter() {defaultValue=-0.25, parameterName=ShadowOffset_X}
  n23 = ScalarParameter() {defaultValue=0.175, parameterName=ShadowOffset_Y}
  n24 = MaterialFunctionCall(
   UVs=Add(
    A=MaterialFunctionCall(UVs=@n20, "Rotation Angle (0-1)"=@n21) {materialFunction=&CustomRotator}.Rotated Values
    B=AppendVector(A=@n22, B=@n23)
   )
   Size=AppendVector(
    A=ScalarParameter() {defaultValue=0.5, parameterName=ShadowBoxSize_X}
    B=ScalarParameter() {defaultValue=0.3333, parameterName=ShadowBoxSize_Y}
   )
   CornerRoundness=ScalarParameter() {defaultValue=0.15, parameterName=ShadowRoundness}
   KeepAspectRatio=StaticBool() {value=true}
  ) {materialFunction=&MF_UI_Box}
  LinearInterpolate(
   A=@n17
   B=LinearInterpolate(A=@n17, B=@n15.Dark, Alpha=LinearInterpolate(B=@n18, Alpha=@n19))
   Alpha=SmoothStep(Value=OneMinus(None=@n24.SDF)) {constMin=0.6}
  )
  NamedRerouteDeclaration(None=@n9.EaseInSine) {name=SelectedAnimCurve, nodeColor=<rgba 0 0 0 1>}
 ]
}
