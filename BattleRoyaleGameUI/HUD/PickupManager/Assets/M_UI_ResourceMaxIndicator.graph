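# UI-domain material for the Battle Royale pickup manager HUD, named as the resource max indicator.
# A rounded-box SDF forms the backing and soft glow, intersected with an angled mask for a line accent.
# Colors mix parameters, shared MPC values and adaptive colors; panned sine scanlines, eased UV gradients and a storm-tech pattern add animated FX.
# Alpha-composite blending uses mix-ratio parameters to shift backing and line toward additive.
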Material /BattleRoyaleGameUI/HUD/PickupManager/Assets/M_UI_ResourceMaxIndicator
graph {
 properties {bAutomaticallySetUsageInEditor=false, blendMode=BLEND_AlphaComposite, materialDomain=MD_UI, usedShadingModels=Default Lit}
 nodes [
  Add()
  ComponentMask(None=CollectionParameter() {collection=/BattleRoyaleGameUI/Shared/MPC_BRGameUI, parameterName=Cyan}) {b=true, g=true, r=true}
  n1 = NamedRerouteUsage()
  LinearInterpolate(
   A=LinearInterpolate(A=NamedRerouteUsage(), B=NamedRerouteUsage(), Alpha=ScalarParameter() {parameterName=BackingColorMixRatio})
   B=LinearInterpolate(A=NamedRerouteUsage(), B=NamedRerouteUsage(), Alpha=ScalarParameter() {defaultValue=0.05, parameterName=FXColorMixRatio})
   Alpha=@n1
  )
  n2 = MaterialFunctionCall(
   Size=Constant() {r=1}
   CornerRoundness=ScalarParameter() {defaultValue=0.1, parameterName=CornerRadius}
   KeepAspectRatio=StaticBool() {value=true}
  ) {materialFunction=&MF_UI_Box}
  n3 = MaterialFunctionCall(
   DistanceField=@n2.SDF
   Threshold=ScalarParameter() {defaultValue=-0.25, parameterName=GlowThreshold}
   Smoothing=ScalarParameter() {defaultValue=0.5, parameterName=GlowSmoothing}
  ) {materialFunction=&MF_UI_AntiAliasing}
  n4 = Multiply(
   A=NamedRerouteDeclaration(None=Saturate(None=@n3.Inner)) {name=Glow, nodeColor=<rgba 0 0 0 1>}
   B=ScalarParameter() {defaultValue=1, parameterName=GlowOpacity}
  )
  n5 = MaterialFunctionCall(DistanceField=@n2.SDF, Threshold=ScalarParameter() {defaultValue=-0.1, parameterName=Threshold}) {materialFunction=&MF_UI_AntiAliasing}
  n6 = ScalarParameter() {parameterName=BackingAdditiveMixRatio}
  n7 = OneMinus(None=ScalarParameter() {defaultValue=0.5, parameterName=AngledLineAdditiveMixRatio})
  n8 = TextureCoordinate()
  n9 = ScalarParameter() {defaultValue=0.0625, parameterName=Offset}
  n10 = Saturate(None=MaterialFunctionCall!(
    UV=Add(A=@n8, B=@n9) {constB=0.1}
    Corner=Constant() {r=2}
    Angle=ScalarParameter() {defaultValue=1.5, parameterName=Angle}
    Flip=Constant() {r=1}
    Falloff
   ) {materialFunction=/FortUILibrary/MaterialFunctions/Shapes/MF_UI_BoxAngle}.Result)
  LinearInterpolate(
   A=Multiply(A=LinearInterpolate(A=@n4, Alpha=@n5.Inner), B=OneMinus(None=@n6))
   B=@n7
   Alpha=NamedRerouteDeclaration(None=Min(A=@n10, B=@n5.Inner)) {name=LineMask, nodeColor=<rgba 0 0 0 1>}
  )
  Multiply(A=@n1, B=MaterialFunctionCall() {materialFunction=/BattleRoyaleGameUI/Shared/FX/MF_UI_BR_StormTechPattern}.Result)
  n11 = NamedRerouteUsage()
  Multiply(A=@n11, B=@n7)
  n12 = VectorParameter() {parameterName=UnselectedColor_A, defaultValue=<rgba 0 0 0.5 1>}
  n13 = LinearInterpolate(
   A=ComponentMask(None=CollectionParameter() {collection=&MPC_BRGameUI, parameterName=Blue}) {b=true, g=true, r=true}
   B=VectorParameter() {parameterName=UnselectedColor_B, defaultValue=<rgba 0 0.4167 1 1>}.RGB
   Alpha=ScalarParameter() {defaultValue=0.5, parameterName=LineColorMixRatio}
  )
  Multiply(A=VertexColor().A, B=Multiply(A=LinearInterpolate(A=@n12.RGB, B=@n13, Alpha=@n11), B=@n11))
  n14 = TextureCoordinate()
  n15 = ComponentMask(None=Sine(None=Panner(Coordinate=@n14) {speedY=0.333}) {period=2}) {g=true}
  n16 = ComponentMask(None=Sine(None=Panner(Coordinate=@n14) {speedY=0.25})) {g=true}
  n17 = Sine(None=Panner(Coordinate=@n14) {speedY=0.4}) {period=0.5}
  n18 = LinearInterpolate(B=Min(A=@n16, B=ComponentMask(None=@n17) {g=true}), Alpha=Min(A=@n16, B=@n15))
  Multiply(A=@n15, B=@n18)
  n19 = MaterialFunctionCall() {materialFunction=&GetUserInterfaceUV}
  n20 = MaterialFunctionCall("Value (0-1)"=ComponentMask(None=@n19.Normalized UV) {g=true}) {materialFunction=&MF_UI_EaseCurves}
  n21 = OneMinus(None=@n20.EaseInOutCubic)
  n22 = Saturate(None=@n20.EaseInQuint)
  n23 = MaterialFunctionCall() {materialFunction=&MF_UI_BR_StormTechPattern}
  n24 = NamedRerouteUsage()
  n25 = ComponentMask(None=Sine(None=@n19.Normalized UV) {period=2}) {r=true}
  n26 = Multiply(
   A=Multiply(A=Multiply(A=@n21, B=@n22), B=LinearInterpolate(Alpha=@n23.Result) {constA=0.5})
   B=Multiply(A=LinearInterpolate(Alpha=@n24) {constA=0.5}, B=@n25)
  )
  n27 = LinearInterpolate(Alpha=MaterialFunctionCall(Speed=Constant() {r=0.5}) {materialFunction=&DebugTimeSine}.Result) {constA=32, constB=64}
  NamedRerouteDeclaration(None=Saturate(None=Multiply(A=@n26, B=@n27) {constB=32})) {name=GradientMask, nodeColor=<rgba 0 0 0 1>}
  NamedRerouteDeclaration(None=Saturate(None=@n18)) {name=Scanlines, nodeColor=<rgba 0 0 0 1>}
  NamedRerouteDeclaration() {name=BlueGlow, nodeColor=<rgba 0 0 0 1>}
  NamedRerouteDeclaration(None=MaterialFunctionCall!(BufferSource) {materialFunction=/BattleRoyaleGameUI/Shared/MF_UI_AdaptiveColors}.Bright) {name=Bright, nodeColor=<rgba 0 0 0 1>}
  NamedRerouteUsage()
  Saturate(None=MaterialFunctionCall("Value (0-1)"=@n25) {materialFunction=&MF_UI_EaseCurves}.EaseInSine)
 ]
}
