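# M_UI_ResourceMaxText is a translucent UI-domain material in the Battle Royale HUD pickup manager, named for resource max text.
# Its graph holds a single constant node of value one, with no connection to a material output shown.
# Automatic usage flags in the editor are turned off, and no parameters, textures or animation appear.
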Material /BattleRoyaleGameUI/HUD/PickupManager/Assets/M_UI_ResourceMaxText
graph {
 properties {bAutomaticallySetUsageInEditor=false, blendMode=BLEND_Translucent, materialDomain=MD_UI, usedShadingModels=Default Lit}
 nodes [Constant() {r=1}]
}
